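# EQS query that picks a destination point for an NPC teleport-forward ability.
# A two-ring donut around the querier generates candidates, projected onto geometry by trace channel.
# Pathfinding with a water-avoiding nav filter and a vertical distance range filter out unsuitable points.
# A dot test favors points along the querier's facing, while negatively weighted 3D distance scoring favors closer ones.
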EnvQuery /NPCLibrary/Abilities/TeleportForward/EQS_NPC_FindTeleportForwardPoint
options [{
  generator {
   name=EnvQueryGenerator_Donut_1
   class=EnvQueryGenerator_Donut
   properties {
    bUseSpiralPattern=false
    center=&EnvQueryContext_Querier
    numberOfRings=2
    pointsPerRing=12
    arcDirection {dirMode=Rotation}
    innerRadius {dataBinding=null, dataField=FloatValue, defaultValue=200}
    outerRadius {dataBinding=null, dataField=FloatValue, defaultValue=400}
    projectionData {
     extentX=0
     navigationFilter=/Game/Athena/AI/Phoebe/NavFilter/NavFilter_Phoebe_NoWater.NavFilter_Phoebe_NoWater_C
     postProjectionVerticalOffset=20
     projectDown=100
     projectUp=100
     serializedChannel=ECC_GameTraceChannel11
     traceChannel=TraceTypeQuery10
     traceMode=GeometryByChannel
    }
   }
  }
  tests [
   EnvQueryTest_Pathfinding_0 {
    class=EnvQueryTest_Pathfinding
    properties {
     filterClass=&NavFilter_Phoebe_AvoidWater_C
     pathFromContext {dataBinding=null, dataField=BoolValue, defaultValue=true}
    }
   }
   EnvQueryTest_Distance_0 {
    class=EnvQueryTest_Distance
    properties {
     distanceTo=&EnvQueryContext_Querier
     filterType=Range
     floatValueMax=500
     floatValueMin=-500
     scoringEquation=Linear
     scoringFactor=1
     testComment=""
     testMode=DistanceZ
     testPurpose=Filter
    }
   }
   EnvQueryTest_Dot_0 {
    class=EnvQueryTest_Dot
    properties {
     clampMaxType=None
     clampMinType=None
     filterType=Minimum
     scoreClampMax=0
     scoreClampMin=0
     scoringFactor=1.5
     testComment=""
     testMode=Dot3D
    }
   }
   EnvQueryTest_Distance_1 {
    class=EnvQueryTest_Distance
    properties {
     distanceTo=&EnvQueryContext_Querier
     filterType=Range
     floatValueMin=0
     scoringEquation=Linear
     scoringFactor=-1
     testComment=""
     testMode=Distance3D
     testPurpose=Score
    }
   }
  ]
 }]
